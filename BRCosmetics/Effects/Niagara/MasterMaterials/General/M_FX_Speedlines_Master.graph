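# Translucent, unlit, two-sided master material for speedline effects on Niagara sprite and mesh particles.
# A scrolling, scalable grayscale speedline texture gets its contrast from a parameterized SmoothStep, then is masked by a flippable V gradient and particle alpha.
# Particle color runs through the shared emissive adjustments function.
# A TexOffset dynamic parameter feeds a MakeFloat2 that does not connect to the texture UVs.
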Material /BRCosmetics/Effects/Niagara/MasterMaterials/General/M_FX_Speedlines_Master
graph {
 properties {
  bUsedWithNiagaraMeshParticles=true
  bUsedWithNiagaraSprites=true
  blendMode=BLEND_Translucent
  pixelDepthOffsetMode=PDOM_Legacy
  shadingModel=MSM_Unlit
  twoSided=true
  usedShadingModels=Unlit
 }
 nodes [
  n1 = ParticleColor()
  MaterialFunctionCall(Emissive=@n1.RGB) {materialFunction=&MF_FX_EmissiveAdjustments}
  MaterialFunctionCall(X=DynamicParameter() {defaultValue=<rgba 0 1 1 1>, paramNames [TexOffset,Param2,Param3,Param4]}.TexOffset) {materialFunction=&MakeFloat2}
  n2 = ScalarParameter() {defaultValue=0.3, parameterName=ContrastMin}
  n3 = ScalarParameter() {defaultValue=0.6, parameterName=ContrastMax}
  n4 = TextureCoordinate()
  n5 = AppendVector(A=ScalarParameter() {defaultValue=1, parameterName=TexScaleX}, B=ScalarParameter() {defaultValue=1, parameterName=TexScaleY})
  n6 = Time()
  n7 = AppendVector(A=ScalarParameter() {defaultValue=2.5, parameterName=TexSpeedX}, B=ScalarParameter() {defaultValue=0.1, parameterName=TexSpeedY})
  n8 = TextureSample(UVs=Add(A=Multiply(A=@n4, B=@n5), B=Multiply(A=@n6, B=@n7)) {constB=0}) {samplerType=SAMPLERTYPE_LinearGrayscale, texture=/BRCosmetics/Effects/Niagara/MasterMaterials/General/Textures/T_SpeedLines}
  n9 = ComponentMask(None=TextureCoordinate()) {g=true}
  n10 = OneMinus(None=@n9)
  n11 = ScalarParameter() {parameterName=FlipGradient}
  Multiply(
   A=@n1.A
   B=Multiply(A=SmoothStep(Min=@n2, Max=@n3, Value=@n8.R) {constMax=0.6, constMin=0.3}, B=LinearInterpolate(A=@n9, B=@n10, Alpha=@n11))
  )
 ]
}
